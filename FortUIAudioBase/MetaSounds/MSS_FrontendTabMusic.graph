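# One-shot MetaSound source that plays a supplied music track for frontend tabs, limited to one instance that replaces the oldest.
# A stereo wave player is scaled by Volume and an interpolated fade gain, then duplicated into the front and rear pairs of quad output.
# Changes to bFadeOut ramp that gain down or back up over FadeOutTime.
# A matching resettable delay triggers OnFinished and updates bIsAlmostFinished after fade-out.
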MetaSoundSource /FortUIAudioBase/MetaSounds/MSS_FrontendTabMusic
inputs [
 bFadeOut {type=Bool}
 FadeOutTime {type=Time}
 Music {type=WaveAsset}
 Start Time {type=Time}
 UE.Source.OnPlay {type=Trigger}
 Volume {type=Float}
]
interfaces [UE.OutputFormat.Quad:1.0, UE.Source.OneShot:1.0, UE.Source:1.1]
outputs [
 bIsAlmostFinished {type=Bool}
 "UE.OutputFormat.Quad.Audio:0" {type=Audio}
 "UE.OutputFormat.Quad.Audio:1" {type=Audio}
 "UE.OutputFormat.Quad.Audio:4" {type=Audio}
 "UE.OutputFormat.Quad.Audio:5" {type=Audio}
 UE.Source.OneShot.OnFinished {type=Trigger}
]
pages [Default {
  input_defaults {FadeOutTime=2, Music=null, Start Time=0, UE.Source.OnPlay=null, Volume=1, bFadeOut=null}
  nodes [
   TriggerOnValueChange_0 = UE.Trigger On Value Change:Bool(Value=^bFadeOut)
   n1 = TriggerCompare.Clamp:Bool(A=^bFadeOut, Compare=@TriggerOnValueChange_0.Trigger) {also {defaults {B="<MetaSoundBuilderResult.SUCCEEDED: 0>"}}}
   n2 = Value.Value:Float(Reset=@n1.False, Set=@n1.True) {also {defaults {Init Value=1, Target Value=0}}}
   n3 = UE.Multiply:Float(AdditionalOperands=^Volume, PrimaryOperand=UE.InterpTo:Audio(Interp Time=^FadeOutTime, Target=@n2.Output Value).Value)
   WavePlayer_0 = UE.Wave Player:Stereo(Wave Asset=^Music, Start Time=^Start Time, Play=^UE.Source.OnPlay) {also {defaults {Loop="<MetaSoundBuilderResult.SUCCEEDED: 0>"}, version [1,1]}}
   Multiply_0 = UE.Multiply:Audio by Float(AdditionalOperands=@n3.Out, PrimaryOperand=@WavePlayer_0.Out Left)
   Multiply_1 = UE.Multiply:Audio by Float(AdditionalOperands=@n3.Out, PrimaryOperand=@WavePlayer_0.Out Right)
   TriggerDelay_0 = UE.Trigger Delay(Delay Time=^FadeOutTime, Reset=@n2.On Reset, In=@n2.On Set)
   ValueValue_0 = Value.Value:Bool(Set=@TriggerDelay_0.Out)
  ]
  links [
   {dst="@UE.OutputFormat.Quad.Audio:0", src=Multiply_0.Out}
   {dst="@UE.OutputFormat.Quad.Audio:4", src=Multiply_0.Out}
   {dst="@UE.OutputFormat.Quad.Audio:1", src=Multiply_1.Out}
   {dst="@UE.OutputFormat.Quad.Audio:5", src=Multiply_1.Out}
   {dst="@UE.Source.OneShot.OnFinished", src=TriggerDelay_0.Out}
   {dst="@bIsAlmostFinished", src=ValueValue_0.Output Value}
  ]
 }]
properties {
 assetImportData=&self:AssetImportData
 bOverrideConcurrency=true
 numChannels=4
 outputFormat=Quad
 sampleRate=48000
 soundClassObject=/Game/Sounds/FortSoundClassesAndModes/Fort_Music_MusicPacks
 concurrencyOverrides {maxCount=1, resolutionRule=StopOldest}
}
